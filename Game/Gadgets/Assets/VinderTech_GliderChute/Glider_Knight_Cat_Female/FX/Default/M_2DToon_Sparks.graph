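# Unlit, masked, two-sided Niagara sprite and mesh particle material for toon sparks on the Knight Cat Female glider.
# A dynamic Time parameter drives a four-column flipbook with swizzlable UVs, hard-thresholded into a cel-style shape.
# Vertex red and a U-axis edge falloff shape the mask, with optional particle-alpha and view-angle fades.
# Particle color feeds HDR emissive adjustment, and a SplineThicken world-position offset is blendable for spline meshes.
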Material /Game/Gadgets/Assets/VinderTech_GliderChute/Glider_Knight_Cat_Female/FX/Default/M_2DToon_Sparks
graph {
 properties {
  bUsedWithNiagaraMeshParticles=true
  bUsedWithNiagaraSprites=true
  blendMode=BLEND_Masked
  pixelDepthOffsetMode=PDOM_Legacy
  shadingModel=MSM_Unlit
  twoSided=true
  usedShadingModels=Unlit
 }
 nodes [
  n1 = TextureCoordinate()
  ComponentMask(None=@n1) {g=true}
  Constant() {r=1}
  Frac(None=Time())
  n2 = ScalarParameter() {defaultValue=10, parameterName=Width}
  LinearInterpolate(
   A=Constant()
   B=MaterialFunctionCall(WidthBase=@n2, WidthTip=@n2) {materialFunction=&SplineThicken}.WorldPositionOffset
   Alpha=ScalarParameter() {parameterName=UseSplineMesh}
  )
  n3 = Constant() {r=1}
  n4 = DynamicParameter() {parameterIndex=1, defaultValue=<rgba 1 0 0 1>, paramNames [RGB,G,B,none]}
  n5 = Constant()
  n6 = VertexColor()
  n7 = ParticleColor()
  n8 = DynamicParameter() {paramNames [Time,Picker,Param3,Param4]}
  n9 = TextureCoordinate()
  n10 = LinearInterpolate(
   A=@n9
   B=MaterialFunctionCall(XY=@n9) {materialFunction=&Swizzle}.YX
   Alpha=ScalarParameter() {defaultValue=1, parameterName=UseSwizzleUV}
  )
  n11 = MaterialFunctionCall(Float3=MaterialFunctionCall(
    "Animation  Phase (0-1)"=@n8.Time
    Number of Rows=@n3
    Number of Columns=Constant() {r=4}
    Texture=TextureObject() {texture=/Game/Weapons/FORT_Melee/Effects/ToonShader/T_2DToon_Flipbook_Sparks}
    UVs=@n10
   ) {materialFunction=&FlipBook}.Result) {materialFunction=&BreakOutFloat3Components}
  n12 = Constant() {r=1}
  n13 = If(
   A=MaterialFunctionCall(A=@n11.R, B=@n11.R, C=@n11.R, Alpha=@n8.Picker) {materialFunction=&Lerp_3Color}.Result
   B=Constant() {r=0.4}
   "A > B"=@n12
   "A == B"=@n12
   "A < B"=Constant()
  )
  n14 = Multiply(A=@n7.A, B=@n13)
  n15 = ScalarParameter() {parameterName=Use Particle Alpha}
  n16 = ComponentMask(None=@n1) {r=true}
  n17 = Multiply(A=@n16, B=OneMinus(None=@n16))
  n18 = Multiply(
   A=Multiply(A=If(A=@n4.RGB, B=@n5, "A > B"=@n6.R, "A == B"=@n6.R, "A < B"=@n6.R), B=LinearInterpolate(A=@n13, B=@n14, Alpha=@n15))
   B=Saturate(None=Add(A=@n17, B=@n17))
  )
  LinearInterpolate(
   A=@n18
   B=Multiply(
    A=@n18
    B=MaterialFunctionCall("FadeSensitivity (optional)"=@n3) {materialFunction=&FadeBasedOnViewAngle}.Result
   )
   Alpha=ScalarParameter() {parameterName=Use Angle and Depth Fade}
  )
  MaterialFunctionCall(Emissive=MaterialFunctionCall(LDR Input=Clamp(None=@n7.RGB), HDR Tint=@n7.RGB) {materialFunction=&DeriveHDRfromLDR}.RemappedRange) {materialFunction=&MF_FX_EmissiveAdjustments}
  Multiply(A=@n10) {constB=-1}
  ScalarParameter() {defaultValue=1}
  StaticBool()
 ]
}
