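# Translucent, unlit, two-sided glow bloom material for Niagara sprites, ribbons and mesh particles in the Knight Cat Female glider effects.
# Particle color passes through an emissive adjustment function to set the glow.
# A UV-centered sphere mask, raised to a power, forms a soft radial falloff multiplied by particle alpha.
# That opacity is also faded by view angle and by depth fade, which the mobile shading path replaces with a constant.
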Material /Game/Gadgets/Assets/VinderTech_GliderChute/Glider_Knight_Cat_Female/FX/Default/M_Glow_Bloom
graph {
 properties {
  bUsedWithNiagaraMeshParticles=true
  bUsedWithNiagaraRibbons=true
  bUsedWithNiagaraSprites=true
  blendMode=BLEND_Translucent
  pixelDepthOffsetMode=PDOM_Legacy
  previewMesh=/BRCosmetics/Gadgets/Assets/VinderTech_GliderChute/Glider_Knight_Cat_Female/FX/Default/SM_Only_Glow_Blades_KC
  shadingModel=MSM_Unlit
  twoSided=true
  usedShadingModels=Unlit
 }
 nodes [
  n1 = ParticleColor()
  MaterialFunctionCall(Emissive=@n1.RGB) {materialFunction=&MF_FX_EmissiveAdjustments}
  n2 = Constant() {r=0.5}
  n3 = SphereMask(A=TextureCoordinate(), B=@n2, Radius=@n2, Hardness=Constant())
  n4 = DepthFade(FadeDistance=ScalarParameter() {defaultValue=20, parameterName=Depth Fade Distance})
  n5 = Constant() {r=1}
  n6 = Constant() {r=5}
  Multiply(
   A=Multiply(A=@n1.A, B=Power(Base=@n3) {constExponent=5})
   B=Multiply(
    A=ShadingPathSwitch(Default=@n4, Deferred=@n4, Mobile=@n5) {desc=Migrated from FeatureLevelSwitch}
    B=MaterialFunctionCall("FadeSensitivity (optional)"=@n6) {materialFunction=&FadeBasedOnViewAngle}.Result
   )
  )
 ]
}
