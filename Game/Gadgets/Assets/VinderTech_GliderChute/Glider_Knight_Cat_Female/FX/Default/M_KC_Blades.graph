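# Unlit, translucent, two-sided material for blade FX on the female Knight Cat glider, usable with Niagara meshes, ribbons and sprites.
# A contrast-boosted Fresnel term blends grey to white, which particle color tints before an emissive adjustment function.
# Particle alpha is multiplied by depth fade and a view-angle fade function.
# The mobile shading path replaces depth fade with a constant.
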Material /Game/Gadgets/Assets/VinderTech_GliderChute/Glider_Knight_Cat_Female/FX/Default/M_KC_Blades
graph {
 properties {
  bUsedWithNiagaraMeshParticles=true
  bUsedWithNiagaraRibbons=true
  bUsedWithNiagaraSprites=true
  blendMode=BLEND_Translucent
  pixelDepthOffsetMode=PDOM_Legacy
  previewMesh=/BRCosmetics/Gadgets/Assets/VinderTech_GliderChute/Glider_Knight_Cat_Female/FX/Default/SM_Knight_Cat_Blade_Low
  shadingModel=MSM_Unlit
  twoSided=true
  usedShadingModels=Unlit
 }
 nodes [
  n1 = Constant3Vector() {constant=<rgba 0.5 0.5 0.5 0>}
  n2 = Constant3Vector() {constant=<rgba 1 1 1 0>}
  n3 = Fresnel(ExponentIn=Constant() {r=1})
  n4 = MaterialFunctionCall(In=@n3, Contrast=Constant() {r=0.5}) {materialFunction=&CheapContrast}
  n5 = Add(A=@n3, B=@n4.Result)
  n6 = Multiply(A=Saturate(None=Multiply(A=@n5, B=@n4.Result))) {constB=0.75}
  n7 = ParticleColor()
  MaterialFunctionCall(Emissive=Multiply(A=LinearInterpolate(A=@n1, B=@n2, Alpha=@n6), B=@n7.RGB)) {materialFunction=&MF_FX_EmissiveAdjustments}
  n8 = DepthFade(FadeDistance=ScalarParameter() {defaultValue=20, parameterName=Depth Fade Distance})
  n9 = Constant() {r=1}
  n10 = Constant()
  Multiply(
   A=@n7.A
   B=Multiply(
    A=ShadingPathSwitch(Default=@n8, Deferred=@n8, Mobile=@n9) {desc=Migrated from FeatureLevelSwitch}
    B=MaterialFunctionCall("FadeSensitivity (optional)"=@n10) {materialFunction=&FadeBasedOnViewAngle}.Result
   )
  )
 ]
}
